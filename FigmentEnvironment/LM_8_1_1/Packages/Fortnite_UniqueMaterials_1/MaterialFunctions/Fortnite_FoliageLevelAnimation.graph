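# Material function that animates one hierarchy level of foliage in wind, rotating branch elements around their pivots.
# Time- and position-driven samples of a turbulence/gust texture perturb the rotation axis, derived from wind and branch direction, and set the angle.
# Optional wind-shelter masks and distance falloff along the branch shape that angle, and model-versus-rotation depth masks the result.
# It outputs world position offset, final position, rotated normal, angle and falloff visualization.
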MaterialFunction /FigmentEnvironment/LM_8_1_1/Packages/Fortnite_UniqueMaterials_1/MaterialFunctions/Fortnite_FoliageLevelAnimation
graph {
 properties {libraryCategoriesText [Fortnite]}
 nodes [
  n1 = FunctionInput() {bUsePreviewValueAsDefault=true, inputName=Optional - world space pixel normal, previewValue=<0 0 1 1>}
  n2 = FunctionInput() {
   bUsePreviewValueAsDefault=true
   inputName=Random Rotation Influence
   inputType=FunctionInput_Scalar
   sortPriority=12
   previewValue=<0.25 0 0 1>
  }
  n3 = Multiply(A=Time() {bIgnorePause=true}, B=FunctionInput() {inputName=TimeRate_CodeReferenced, inputType=FunctionInput_Scalar})
  n4 = FunctionInput() {inputName=TimeOffset_CodeControlled, inputType=FunctionInput_Scalar}
  n5 = Constant2Vector() {g=1.125, r=1}
  n6 = FunctionInput() {inputName=Pivot Position, sortPriority=-5}
  n7 = FunctionInput() {inputName=X-Axis Texture And X Extent Alpha, inputType=FunctionInput_Vector4, sortPriority=-4}
  n8 = ComponentMask(None=@n7) {b=true, g=true, r=true}
  n9 = ComponentMask(None=@n7) {a=true}
  n10 = Add(A=@n6, B=Multiply(A=@n8, B=@n9))
  n11 = VectorParameter() {group=Wind Settings, parameterName=Texture Scale, defaultValue=<rgba -512 -512 0 1>}
  n12 = Add(
   A=Multiply(A=Add(A=@n3, B=@n4), B=@n5)
   B=Divide(A=ComponentMask(None=@n10) {g=true, r=true}, B=ComponentMask(None=@n11.RGB) {g=true, r=true}) {constB=512}
  )
  n13 = FunctionInput() {
   bUsePreviewValueAsDefault=true
   inputName=Random Rotation Texture Sample Scale
   inputType=FunctionInput_Scalar
   sortPriority=11
   previewValue=<250 0 0 1>
  }
  n14 = TextureObject() {
   samplerType=SAMPLERTYPE_LinearColor
   texture=/Engine/Functions/Engine_MaterialFunctions02/ExampleContent/Textures/WindTurblenceVectorAndGustMagnitude
  }
  n15 = FunctionInput(Preview=@n14) {bUsePreviewValueAsDefault=true, inputName=Optional - Wind Turbulence and Gust Mag, inputType=FunctionInput_Texture2D, sortPriority=99}
  n16 = TextureSample!(UVs=Frac(None=Divide(A=@n12, B=@n13)), Tex=@n15, Level=Constant(), Apply View MipBias) {
   mipValueMode=TMVM_MipLevel
   samplerType=SAMPLERTYPE_LinearColor
   texture=&WindTurblenceVectorAndGustMagnitude
  }
  n17 = Multiply(A=@n2, B=ConstantBiasScale(None=@n16.RGB) {bias=-0.5, scale=2})
  n18 = WorldPosition() {worldPositionShaderOffset=WPT_ExcludeAllShaderOffsets}
  n19 = FunctionInput() {bUsePreviewValueAsDefault=true, inputName=Wind Direction, sortPriority=1000, previewValue=<1 0 0 1>}
  n20 = CrossProduct(A=Normalize(VectorInput=Subtract(A=@n18.XYZ, B=@n6)), B=@n19)
  n21 = Normalize(VectorInput=Add(A=@n17, B=@n19))
  n22 = Power(Base=DotProduct(A=@n19, B=@n8)) {constExponent=5}
  n23 = Constant3Vector() {constant=<rgba 0 0 -0.2 0>}
  n24 = Normalize(VectorInput=Add(A=@n8, B=Multiply(A=@n22, B=@n23)))
  n25 = FunctionInput(Preview=StaticBool()) {bUsePreviewValueAsDefault=true, inputName=UseWrapAroundLogic, inputType=FunctionInput_StaticBool}
  n26 = StaticSwitch(True=Normalize(VectorInput=Add(A=@n17, B=@n20)), False=Normalize(VectorInput=CrossProduct(A=@n21, B=@n24)), Value=@n25) {desc=Axis}
  n27 = FunctionInput() {bUsePreviewValueAsDefault=true, inputName=Wind Gust Angle Rotation, inputType=FunctionInput_Scalar, previewValue=<0.1 0 0 1>}
  n28 = ComponentMask(None=MaterialFunctionCall() {materialFunction=&ComponentPivotLocation}.Result) {g=true, r=true}
  n29 = WorldPosition() {worldPositionShaderOffset=WPT_ExcludeAllShaderOffsets}
  n30 = FunctionInput() {bUsePreviewValueAsDefault=true, inputName=Pivot of point 1 level higher in the hierarchy}
  n31 = Subtract(
   A=FunctionInput() {bUsePreviewValueAsDefault=true, inputName=Model Depth, inputType=FunctionInput_Scalar, sortPriority=99}
   B=FunctionInput() {bUsePreviewValueAsDefault=true, inputName=rotation Depth, inputType=FunctionInput_Scalar, sortPriority=99}
  ) {desc=0 matches level. positive is per pivot. everything else is per vertex}
  n32 = Clamp(None=@n31)
  n33 = FunctionInput(Preview=StaticBool() {value=true}) {bUsePreviewValueAsDefault=true, inputName=Use Pivots To Sample Wind Intensity At Higher Level Branches, inputType=FunctionInput_StaticBool}
  n34 = Subtract(A=StaticSwitch(True=LinearInterpolate(A=@n29.XYZ, B=@n30, Alpha=@n32), False=@n29.XYZ, Value=@n33), B=@n6)
  n35 = FunctionInput() {
   bUsePreviewValueAsDefault=true
   inputName=Vertex Phase Offset X Axis
   inputType=FunctionInput_Scalar
   sortPriority=100
   previewValue=<100 0 0 1>
  }
  n36 = Divide(A=DotProduct(A=@n34, B=@n8), B=@n35)
  n37 = Constant()
  n38 = FunctionInput(Preview=StaticBool()) {bUsePreviewValueAsDefault=true, inputName=Offset AnimPhase Via X Axis Direction, inputType=FunctionInput_StaticBool, sortPriority=100}
  n39 = Multiply(A=@n5, B=Add(A=Multiply(A=@n28) {constB=0.01}, B=StaticSwitch(True=@n36, False=@n37, Value=@n38)))
  n40 = TextureSample(UVs=Frac(None=Add(A=@n12, B=@n39)), Tex=@n15, Level=Constant()) {
   mipValueMode=TMVM_MipLevel
   samplerType=SAMPLERTYPE_LinearColor
   texture=&WindTurblenceVectorAndGustMagnitude
  }
  n41 = FunctionInput() {bUsePreviewValueAsDefault=true, inputName=Wind Gust Offset, inputType=FunctionInput_Scalar, previewValue=<-0.125 0 0 1>}
  n42 = Multiply(A=@n27, B=Add(A=@n40.A, B=@n41) {constB=-0.25})
  n43 = Abs(None=DotProduct(A=@n19, B=@n8))
  n44 = FunctionInput() {bUsePreviewValueAsDefault=true, inputName=Advanced - Parent Rotation Angle Influence, inputType=FunctionInput_Scalar, sortPriority=10}
  n45 = Add(A=@n44, B=@n42)
  n46 = FunctionInput() {
   bUsePreviewValueAsDefault=true
   description="Scale And then Bias the results from the dot product of the wind vector and the element axis vector. The defaults make sure that the wind fully affects branches facing the wind while leaving the other branches unaffected. "
   inputName=Wind Shelter Scale
   inputType=FunctionInput_Scalar
   sortPriority=50
   previewValue=<-2 0 0 1>
  }
  n47 = DotProduct(A=@n24, B=@n19)
  n48 = FunctionInput() {
   bUsePreviewValueAsDefault=true
   description="Scale And then Bias the results from the dot product of the wind vector and the element axis vector. The defaults make sure that the wind fully affects branches facing the wind while leaving the other branches unaffected. "
   inputName=Wind Shelter Bias
   inputType=FunctionInput_Scalar
   sortPriority=50
   previewValue=<-1 0 0 1>
  }
  n49 = Multiply(A=@n45, B=Add(A=Multiply(A=@n46, B=@n47), B=@n48)) {desc=Anim * Directional Wind Mask Settings}
  n50 = FunctionInput(Preview=StaticBool() {value=true}) {bUsePreviewValueAsDefault=true, inputName="Use \"Wind Shelter\" Directional Masks", inputType=FunctionInput_StaticBool, sortPriority=100}
  n51 = FunctionInput(Preview=StaticBool() {value=true}) {
   bUsePreviewValueAsDefault=true
   description="Defaults to true and outputs the correct World Position Offset and Rotated Normal. "
   inputName=Enable
   inputType=FunctionInput_StaticBool
   sortPriority=-10
  }
  n52 = FunctionOutput(None=StaticSwitch(True=StaticSwitch(True=@n49, False=@n45, Value=@n50), False=Constant(), Value=@n51)) {outputName=Advanced - Rotation Angle Animation, sortPriority=50}
  n53 = Clamp(None=DotProduct(A=@n34, B=@n8)) {clampMode=CMODE_ClampMin}
  n54 = FunctionInput() {
   bUsePreviewValueAsDefault=true
   inputName=Percentage of Branch Length For Harsh Falloff
   inputType=FunctionInput_Scalar
   sortPriority=201
   previewValue=<0.01 0 0 1>
  }
  n55 = Multiply(A=@n9, B=@n54)
  n56 = FunctionInput(Preview=StaticBool()) {bUsePreviewValueAsDefault=true, inputName=Control Harsh Falloff Around Root, inputType=FunctionInput_StaticBool, sortPriority=200}
  n57 = StaticSwitch(True=Subtract(A=@n53, B=@n55), False=@n53, Value=@n56)
  n58 = FunctionInput() {bUsePreviewValueAsDefault=true, inputName=Motion Dampening Falloff Radius, inputType=FunctionInput_Scalar, previewValue=<0.25 0 0 1>}
  n59 = Multiply(A=@n58, B=@n9)
  n60 = FunctionInput(Preview=StaticBool() {value=true}) {bUsePreviewValueAsDefault=true, inputName=Make Motion Falloff Element Size Specific, inputType=FunctionInput_StaticBool}
  n61 = Clamp(None=Divide(A=@n57, B=StaticSwitch(True=@n59, False=@n58, Value=@n60)))
  n62 = StaticBool()
  n63 = StaticSwitch(
   True=Multiply(A=@n42, B=OneMinus(None=@n43))
   False=StaticSwitch(
    True=@n52
    False=Multiply(A=@n52, B=@n61) {desc=Distance Mask}
    Value=FunctionInput(Preview=@n62) {bUsePreviewValueAsDefault=true, inputName=RigidAnimation?, inputType=FunctionInput_StaticBool}
   )
   Value=@n25
  ) {desc=Angle}
  n64 = Clamp(None=Add(A=@n31))
  n65 = LinearInterpolate(
   A=@n1
   B=Add(
    A=@n1
    B=MaterialFunctionCall(Rotation Axis=@n26, Rotation Angle=@n63, World Space Vertex Normals=@n1) {materialFunction=&FixRotateAboutAxisNormals}.New World Space Vertex Normal
   )
   Alpha=@n64
  )
  n66 = FunctionInput(Preview=StaticBool() {value=true}) {
   bUsePreviewValueAsDefault=true
   description="Defaults to true. "
   inputName=Enable Normal Rotation
   inputType=FunctionInput_StaticBool
   sortPriority=-9
  }
  FunctionOutput(None=StaticSwitch(True=StaticSwitch(True=@n65, False=@n1, Value=@n66), False=@n1, Value=@n51)) {outputName=Rotated World Space Pixel Normal, sortPriority=10}
  FunctionOutput(None=@n57) {outputName=Visualize harsh falloff, sortPriority=10}
  n67 = WorldPosition() {worldPositionShaderOffset=WPT_ExcludeAllShaderOffsets}
  n68 = StaticSwitch(
   True=Multiply(A=RotateAboutAxis(NormalizedRotationAxis=@n26, RotationAngle=@n63, PivotPoint=@n6, Position=@n67.XYZ), B=@n64)
   False=Constant()
   Value=@n51
  )
  FunctionOutput(None=Add(A=FunctionOutput(None=@n68) {outputName=Add to WPO}, B=WorldPosition() {worldPositionShaderOffset=WPT_ExcludeAllShaderOffsets}.XYZ)) {outputName=Final World Position, sortPriority=10}
  SphereMask(Radius=@n55) {hardnessPercent=0}
 ]
}
